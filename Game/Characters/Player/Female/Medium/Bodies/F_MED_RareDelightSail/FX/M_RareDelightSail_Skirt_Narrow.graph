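# Translucent, two-sided unlit effect material for the narrow skirt of the RareDelightSail player character, with music-reactive opacity.
# Offset noise and magnetic-wave samples split into RGB for chromatic fringing, then mix with tinted gradients into oversaturated, time-of-day-scaled emissive.
# Crossed panning flare rays, smoke and UV edge fades form the alpha mask, softened by view-angle and depth fades on non-mobile paths.
# Event audio amplitude or spectrum, whichever is higher, drives opacity.
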Material /Game/Characters/Player/Female/Medium/Bodies/F_MED_RareDelightSail/FX/M_RareDelightSail_Skirt_Narrow
graph {
 properties {
  bUsedWithMeshDeformer=true
  bUsedWithMorphTargets=true
  bUsedWithNiagaraMeshParticles=true
  bUsedWithNiagaraSprites=true
  bUsedWithSkeletalMesh=true
  blendMode=BLEND_Translucent
  pixelDepthOffsetMode=PDOM_Legacy
  shadingModel=MSM_Unlit
  twoSided=true
  usedShadingModels=Unlit
 }
 nodes [
  n1 = VectorParameter() {parameterName=Color_Morning, defaultValue=<rgba 0.2 0.2 0.2 1>}
  n2 = Constant() {r=1}
  n3 = VectorParameter() {parameterName=Color_Evening, defaultValue=<rgba 0.3 0.3 0.3 1>}
  n4 = VectorParameter() {parameterName=Color_Night, defaultValue=<rgba 1 1 1 0.8>}
  n5 = NamedRerouteUsage()
  n6 = NamedRerouteUsage()
  n7 = Desaturation(None=Multiply(A=Add(A=@n5, B=@n6) {constB=0}, B=Constant3Vector() {constant=<rgba 10 4.1667 8.0556 1>}), Fraction=Constant() {r=-0.5})
  MaterialFunctionCall(Emissive=Multiply(
    A=MaterialFunctionCall(Color_Morning=@n1.RGB, Color_Day=@n2, Color_Evening=@n3.RGB, Color_Night=@n4.RGB) {materialFunction=&Color_Time_of_Day}.Result
    B=Multiply(A=@n7) {constB=0.1}
   )) {materialFunction=&MF_FX_EmissiveAdjustments}
  n8 = ComponentMask(None=TextureCoordinate()) {g=true}
  Multiply(A=Power(Base=OneMinus(None=@n8)) {constExponent=5}) {constB=0.1}
  n9 = Constant() {r=1}
  n10 = Multiply(
   A=NamedRerouteUsage()
   B=MaterialFunctionCall(Value_Morning=@n1.A, Value_Day=Constant() {r=1}, Value_Evening=@n3.A, Value_Night=@n4.A) {materialFunction=&Scalar_Time_of_Day}.Result
  )
  n11 = CollectionParameter() {collection=&FortniteMaterialParameters, parameterName=EventAudio_Amplitude}
  n12 = CollectionParameter() {collection=&FortniteMaterialParameters, parameterName=EventAudio_Spectrum}
  n13 = Max(A=@n11, B=@n12)
  Multiply(A=Multiply(A=Multiply(A=@n10, B=@n9), B=Clamp(None=@n13)), B=ScalarParameter() {defaultValue=1, parameterName=Music Skirt Opacity})
  n14 = TextureCoordinate() {uTiling=0.5, vTiling=0.1}
  n15 = Panner(Coordinate=Add(A=@n14, B=Constant2Vector() {g=0.05})) {speedX=0.1}
  n16 = Panner(Coordinate=Add(A=@n14, B=Constant2Vector() {g=0.1})) {speedX=-0.1}
  n17 = Power(Base=Multiply(
    A=TextureSample(UVs=@n15) {texture=/Game/Characters/Player/Female/Medium/Bodies/F_MED_Buffet_Shard/FX/T_Flare_Ray}.RGB
    B=TextureSample(UVs=@n16) {texture=&T_Flare_Ray}.RGB
   )) {constExponent=1.5}
  n18 = ComponentMask(None=TextureCoordinate()) {g=true}
  n19 = Power(Base=ComponentMask(None=TextureCoordinate()) {g=true}) {constExponent=1}
  n20 = Multiply(A=Power(Base=OneMinus(None=@n18)) {constExponent=1}, B=Saturate(None=Multiply(A=@n19) {constB=10}))
  n21 = LinearInterpolate(Alpha=TextureSample() {texture=&T_Inky_Smoke_Tile}.R) {constA=0.2, constB=1.5}
  n22 = ComponentMask(None=TextureCoordinate()) {r=true}
  n23 = Multiply(A=@n22, B=OneMinus(None=@n22))
  n24 = Multiply(A=Multiply(A=Multiply(A=@n17) {constB=0.5}, B=Multiply(A=@n20, B=@n21)), B=Saturate(None=Multiply(A=@n23) {constB=2.5}))
  n25 = TextureCoordinate() {uTiling=2, vTiling=0.6}
  n26 = ComponentMask(None=@n25) {r=true}
  n27 = OneMinus(None=ComponentMask(None=@n25) {g=true})
  n28 = Panner(Coordinate=Add(A=AppendVector(A=@n26, B=@n27)) {constB=-0.4}) {speedX=0.3}
  n29 = Constant() {r=1}
  n30 = Multiply(
   A=Multiply(
    A=Multiply(A=@n24) {constB=15}
    B=TextureSample(UVs=@n28) {samplerSource=SSM_Wrap_WorldGroupSettings, texture=/Game/Effects/Fort_Effects/Textures/T_Smoke_Ring_Mask2}.R
   )
   B=ShadingPathSwitch(
    Default=MaterialFunctionCall() {materialFunction=&FadeBasedOnViewAngle}.Result
    Forward=@n29
    Mobile=@n29
   )
  )
  n31 = DepthFade(Opacity=@n30) {fadeDistanceDefault=30}
  NamedRerouteDeclaration(None=ShadingPathSwitch(Default=@n31, Deferred=@n31, Forward=@n30, Mobile=@n30)) {name=Alpha_Mask, nodeColor=<rgba 1 0.4938 0.0916 1>}
  n32 = TextureCoordinate()
  n33 = Constant2Vector() {g=1.6, r=1}
  n34 = Desaturation(
   None=TextureSample(UVs=MaterialFunctionCall(UVs=@n32, Texture Scale=@n33) {materialFunction=&ScaleUVsByCenter}.UVs) {texture=/Game/Effects/Fort_Effects/Textures/Grad/T_Halo_Corona_Gradient}.RGB
   Fraction=Constant() {r=1}
  )
  n35 = Constant3Vector() {constant=<rgba 1 0.6597 0 1>}
  n36 = TextureCoordinate() {uTiling=0.2, vTiling=0.2}
  n37 = Constant() {r=0.005}
  n38 = Panner(Coordinate=Add(A=@n36, B=@n37) {constB=0.005}) {speedX=0.05, speedY=-0.1}
  n39 = Panner(Coordinate=@n36) {speedX=0.05, speedY=-0.1}
  n40 = Panner(Coordinate=Subtract(A=@n36, B=@n37)) {speedX=0.05, speedY=-0.1}
  n41 = Power(Base=MaterialFunctionCall(
    X=TextureSample(UVs=@n38) {texture=/Game/Global/Textures/Default/Noise/T_FireNoise_01}.B
    Y=TextureSample(UVs=@n39) {texture=&T_FireNoise_01}.B
    Z=TextureSample(UVs=@n40) {texture=&T_FireNoise_01}.B
   ) {materialFunction=&MakeFloat3}.Result) {constExponent=1.5}
  n42 = Multiply(A=Desaturation(None=Multiply(A=@n41, B=@n41), Fraction=Constant())) {constB=3}
  NamedRerouteDeclaration(None=Add(A=Multiply(A=@n34, B=@n35), B=Multiply(A=@n42) {constB=5})) {name=Texture_Panner01, nodeColor=<rgba 0 0.0132 1 1>}
  n43 = TextureCoordinate() {uTiling=0.3, vTiling=0.4}
  n44 = Constant() {r=0.003}
  n45 = Add(A=@n43, B=@n44) {constB=0.005}
  n46 = Constant2Vector() {r=0.5}
  n47 = TextureSample(UVs=Panner(Coordinate=Add(A=@n45, B=@n46)) {speedY=-0.08}) {texture=/Game/Athena/Environments/Nexus/Textures/T_ZeroPoint_MagneticWaves}
  n48 = TextureSample(UVs=Panner(Coordinate=Add(A=@n43, B=@n46)) {speedY=-0.08}) {texture=&T_ZeroPoint_MagneticWaves}
  n49 = Subtract(A=@n43, B=@n44)
  n50 = TextureSample(UVs=Panner(Coordinate=Add(A=@n49, B=@n46)) {speedY=-0.08}) {texture=&T_ZeroPoint_MagneticWaves}
  n51 = Power(Base=Desaturation(
    None=MaterialFunctionCall(X=@n47.R, Y=@n48.R, Z=@n50.R) {materialFunction=&MakeFloat3}.Result
    Fraction=Constant()
   )) {constExponent=1.5}
  NamedRerouteDeclaration(None=Multiply(A=Multiply(A=@n51) {constB=50}, B=Constant3Vector() {constant=<rgba 0.6851 0.4896 1 0>})) {name=Texture_Panner02, nodeColor=<rgba 0 0.0132 1 1>}
 ]
}
