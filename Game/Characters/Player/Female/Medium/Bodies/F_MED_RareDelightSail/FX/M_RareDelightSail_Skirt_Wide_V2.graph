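# Translucent unlit, two-sided material for the RareDelightSail character's wide skirt effect, with audio-reactive glow.
# Three panning caustic, magnetic-wave and smoke layers use noise distortion, vertex-red UV offsets and triple offset sampling for color fringing.
# Event audio amplitude brightens two layers; a curve-atlas gradient and time-of-day parameters tint the emissive.
# Opacity combines time-of-day values, a music opacity parameter and depth and view-angle fades, simplified on forward and mobile.
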Material /Game/Characters/Player/Female/Medium/Bodies/F_MED_RareDelightSail/FX/M_RareDelightSail_Skirt_Wide_V2
graph {
 properties {
  bUsedWithMeshDeformer=true
  bUsedWithNiagaraMeshParticles=true
  bUsedWithSkeletalMesh=true
  blendMode=BLEND_Translucent
  pixelDepthOffsetMode=PDOM_Legacy
  shadingModel=MSM_Unlit
  twoSided=true
  usedShadingModels=Unlit
 }
 nodes [
  n1 = CollectionParameter() {collection=&FortniteMaterialParameters, parameterName=EventAudio_Amplitude}
  n2 = CollectionParameter() {collection=&FortniteMaterialParameters, parameterName=EventAudio_Spectrum}
  Clamp(None=Max(A=@n1, B=@n2))
  CollectionParameter() {collection=&FortniteMaterialParameters, parameterName=EventAudio_Spectrum}
  DynamicParameter() {paramNames [Param1, Bright Intensity, Param3, Param4]}
  n3 = ComponentMask(None=TextureCoordinate()) {g=true}
  n4 = Saturate(None=Multiply(A=Power(Base=@n3)) {constB=10})
  n5 = CurveAtlasRowParameter(CurveTime=ComponentMask(None=TextureCoordinate()) {g=true}) {
   atlas=&/CVAtlas_RareDelightSail
   curve=&/Curve_RareDelightSail_Color
   parameterName=Global_Color
  }
  n6 = Multiply(A=@n5)
  n7 = Constant() {r=-2}
  n8 = Multiply(A=Multiply(A=Desaturation(None=@n6, Fraction=@n7)) {constB=4}, B=Constant3Vector() {constant=<rgba 1 0.2917 0.2993 1>})
  n9 = Add(A=NamedRerouteUsage(), B=Add(A=NamedRerouteUsage(), B=Add(B=@n8)))
  n10 = Multiply(A=Desaturation(None=Multiply(A=@n9), Fraction=Constant() {r=-0.2})) {constB=0.2}
  n11 = VectorParameter() {parameterName=Color_Morning, defaultValue=<rgba 0.2 0.2 0.2 1>}
  n12 = Constant() {r=1}
  n13 = VectorParameter() {parameterName=Color_Evening, defaultValue=<rgba 0.3 0.3 0.3 1>}
  n14 = VectorParameter() {parameterName=Color_Night, defaultValue=<rgba 1 1 1 0.8>}
  MaterialFunctionCall(Emissive=Multiply(
    A=Multiply(A=@n4, B=@n10) {constA=1}
    B=MaterialFunctionCall(Color_Morning=@n11.RGB, Color_Day=@n12, Color_Evening=@n13.RGB, Color_Night=@n14.RGB) {materialFunction=&Color_Time_of_Day}.Result
   )) {materialFunction=&MF_FX_EmissiveAdjustments}
  MaterialFunctionCall(Speed=Constant() {r=3}) {materialFunction=&DebugTimeSine}
  MaterialFunctionCall(Speed=Constant() {r=3}) {materialFunction=&DebugTimeSine}
  n15 = CollectionParameter() {collection=&FortniteMaterialParameters, parameterName=EventAudio_Amplitude}
  Max(A=@n15) {constB=0}
  n16 = Multiply(A=NamedRerouteUsage()) {constB=14}
  n17 = DepthFade(Opacity=@n16) {fadeDistanceDefault=10}
  n18 = MaterialFunctionCall() {materialFunction=&FadeBasedOnViewAngle}
  n19 = Constant() {r=1}
  n20 = Multiply(
   A=Multiply(
    A=ShadingPathSwitch(Default=@n17, Deferred=@n17, Forward=@n16, Mobile=@n16)
    B=ShadingPathSwitch(Default=@n18.Result, Forward=@n19, Mobile=@n19)
   )
   B=Constant() {r=1}
  )
  n21 = MaterialFunctionCall(Value_Morning=@n11.A, Value_Day=Constant() {r=1}, Value_Evening=@n13.A, Value_Night=@n14.A) {materialFunction=&Scalar_Time_of_Day}
  Multiply(A=Multiply(A=Multiply(A=@n20, B=@n21.Result)), B=ScalarParameter() {defaultValue=1, parameterName=Music Skirt Opacity})
  n22 = Multiply(A=TextureCoordinate() {vTiling=1.5}, B=AppendVector(A=ScalarParameter() {defaultValue=0.8, parameterName=U_Scale}, B=Constant() {r=1}))
  n23 = Multiply(A=TextureSample(UVs=Panner(Coordinate=@n22) {speedY=-0.1}) {samplerType=SAMPLERTYPE_Grayscale, texture=&LowResBlurredNoise}.R)
  n24 = TextureSample(UVs=Panner(Coordinate=TextureCoordinate() {vTiling=2.5}) {speedX=-0.5, speedY=-0.1}) {texture=&T_Inky_Smoke_Tile}
  NamedRerouteDeclaration(None=Add(A=Multiply(A=@n23) {constB=0.07}, B=Multiply(A=@n24.R) {constB=0.07}) {constB=0}) {name=Disturb_Noise, nodeColor=<rgba 0.7698 0 1 1>}
  n25 = ScalarParameter() {defaultValue=0.8, parameterName=U_Scale}
  n26 = Constant() {r=1}
  n27 = Add(
   A=Multiply(A=TextureCoordinate() {uTiling=0.84, vTiling=0.5}, B=AppendVector(A=@n25, B=@n26))
   B=AppendVector(A=Constant(), B=VertexColor().R)
  )
  n28 = Constant() {r=0.004}
  n29 = NamedRerouteUsage()
  n30 = Constant2Vector() {g=-0.1, r=-0.03}
  n31 = Panner(Coordinate=Add(A=Add(A=@n27, B=@n28) {constB=0.005}, B=@n29), Speed=@n30) {speedX=0.05, speedY=-0.1}
  n32 = Panner(Coordinate=Add(A=@n27, B=@n29), Speed=@n30) {speedX=0.05, speedY=-0.1}
  n33 = Panner(Coordinate=Add(A=Subtract(A=@n27, B=@n28), B=@n29), Speed=@n30) {speedX=0.05, speedY=-0.1}
  n34 = Power(Base=MaterialFunctionCall(
    X=TextureSample(UVs=@n31) {samplerType=SAMPLERTYPE_LinearGrayscale, texture=/Game/Global/Textures/Default/Caustics/T_Nosie_Caustics_Grayscale}.R
    Y=TextureSample(UVs=@n32) {samplerType=SAMPLERTYPE_LinearGrayscale, texture=&T_Nosie_Caustics_Grayscale}.R
    Z=TextureSample(UVs=@n33) {samplerType=SAMPLERTYPE_LinearGrayscale, texture=&T_Nosie_Caustics_Grayscale}.R
   ) {materialFunction=&MakeFloat3}.Result) {constExponent=0.5}
  n35 = Multiply(A=@n34, B=@n34)
  n36 = Constant() {r=-1}
  NamedRerouteDeclaration(None=Multiply(A=Desaturation(None=@n35, Fraction=@n36), B=Constant3Vector() {constant=<rgba 3 1.4625 0.4805 1>}) {constB=3}) {name=Panner_texture_01, nodeColor=<rgba 0 0.0132 1 1>}
  n37 = Constant() {r=1.5}
  n38 = Constant() {r=1}
  n39 = QualitySwitch(Default=@n37, Low=@n38, High=@n37, Medium=@n38, Epic=@n37)
  n40 = Power(Base=@n15) {constExponent=3}
  n41 = Constant3Vector() {constant=<rgba 1 0.6433 0.3125 1>}
  n42 = ScalarParameter() {defaultValue=0.8, parameterName=U_Scale}
  n43 = Constant() {r=1}
  n44 = Add(
   A=Multiply(A=TextureCoordinate() {uTiling=0.8, vTiling=0.5}, B=AppendVector(A=@n42, B=@n43))
   B=AppendVector(A=Constant(), B=VertexColor().R)
  )
  n45 = Constant() {r=0.005}
  n46 = Constant2Vector() {g=-0.35, r=0.03}
  n47 = Panner(Coordinate=Add(A=Add(A=@n44, B=@n45) {constB=0.005}, B=@n29), Speed=@n46) {speedX=0.05, speedY=-0.06}
  n48 = Panner(Coordinate=Add(A=@n44, B=@n29), Speed=@n46) {speedX=0.05, speedY=-0.06}
  n49 = Panner(Coordinate=Add(A=Subtract(A=@n44, B=@n45), B=@n29), Speed=@n46) {speedX=0.05, speedY=-0.06}
  n50 = LinearInterpolate(
   A=MaterialFunctionCall(
    X=TextureSample(UVs=@n47) {texture=/Game/Athena/Environments/Nexus/Textures/T_ZeroPoint_MagneticWaves}.R
    Y=TextureSample(UVs=@n48) {texture=&T_ZeroPoint_MagneticWaves}.R
    Z=TextureSample(UVs=@n49) {texture=&T_ZeroPoint_MagneticWaves}.R
   ) {materialFunction=&MakeFloat3}.Result
   B=TextureSample(UVs=@n48) {texture=/Game/Athena/Environments/Nexus/Textures/T_ZeroPoint_MagneticWaves_Blurred}.R
  ) {constAlpha=0.3}
  n51 = Multiply(A=Desaturation(None=Power(Base=@n50) {constExponent=0.5}, Fraction=Constant() {r=-1})) {constB=4}
  NamedRerouteDeclaration(None=Multiply(A=LinearInterpolate(B=@n39, Alpha=@n40) {constA=0.5, constB=1.5}, B=Multiply(A=@n41, B=@n51))) {name=Panner_texture_02, nodeColor=<rgba 0 0.0132 1 1>}
  n52 = Constant() {r=1.5}
  n53 = Constant() {r=1}
  n54 = QualitySwitch(Default=@n52, Low=@n53, High=@n52, Medium=@n53, Epic=@n52)
  n55 = ScalarParameter() {defaultValue=0.8, parameterName=U_Scale}
  n56 = Constant() {r=1}
  n57 = Add(A=Multiply(A=TextureCoordinate() {uTiling=2, vTiling=0.7}, B=AppendVector(A=@n55, B=@n56)), B=AppendVector(A=Constant(), B=VertexColor().R))
  n58 = Constant() {r=0.007}
  n59 = TextureCoordinate() {uTiling=0.5, vTiling=2}
  n60 = ScalarParameter() {defaultValue=0.8, parameterName=U_Scale}
  n61 = Constant() {r=1}
  n62 = Panner(Coordinate=Multiply(
    A=MaterialFunctionCall(XY=@n59) {materialFunction=&Swizzle}.YX
    B=AppendVector(A=@n60, B=@n61)
   ))
  n63 = Multiply(A=ComponentMask(None=TextureSample(UVs=@n62) {samplerType=SAMPLERTYPE_Grayscale, texture=&LowResBlurredNoise}.RGB) {g=true, r=true}) {constB=0.2}
  n64 = Constant2Vector() {r=0.5}
  n65 = Add(A=Add(A=Add(A=@n57, B=@n58) {constB=0.005}, B=@n63), B=@n64)
  n66 = Constant2Vector() {g=-0.5}
  n67 = Add(A=Add(A=@n57, B=@n63), B=@n64)
  n68 = Constant() {r=0.003}
  n69 = Add(A=Add(A=Subtract(A=@n57, B=@n68), B=@n63), B=@n64)
  n70 = MaterialFunctionCall(
   X=TextureSample(UVs=Panner(Coordinate=@n65, Speed=@n66) {speedY=-0.06}) {texture=&T_Inky_Smoke_Tile}.B
   Y=TextureSample(UVs=Panner(Coordinate=@n67, Speed=@n66) {speedY=-0.06}) {texture=&T_Inky_Smoke_Tile}.B
   Z=TextureSample(UVs=Panner(Coordinate=@n69, Speed=@n66) {speedY=-0.06}) {texture=&T_Inky_Smoke_Tile}.B
  ) {materialFunction=&MakeFloat3}
  n71 = Constant() {r=-5}
  n72 = Multiply(A=Power(Base=Desaturation(None=@n70.Result, Fraction=@n71)) {constExponent=1})
  n73 = Constant3Vector() {constant=<rgba 2.5802 1.6094 3 1>}
  NamedRerouteDeclaration(None=Multiply(A=LinearInterpolate(B=@n54, Alpha=@n40) {constA=0.8, constB=2}, B=Multiply(A=@n72, B=@n73))) {name=Panner_texture_03, nodeColor=<rgba 0 0.0132 1 1>}
  n74 = OneMinus(None=ComponentMask(None=TextureCoordinate()) {g=true})
  n75 = TextureCoordinate()
  n76 = Power(Base=OneMinus(None=ComponentMask(None=@n75) {g=true})) {constExponent=1}
  n77 = ComponentMask(None=TextureCoordinate()) {g=true}
  n78 = Saturate(None=Multiply(A=Power(Base=@n77) {constExponent=2.8}) {constB=10})
  n79 = TextureCoordinate() {uTiling=1.5}
  n80 = AppendVector(A=ScalarParameter() {defaultValue=0.8, parameterName=U_Scale}, B=Constant() {r=1})
  n81 = Constant()
  n82 = VertexColor()
  n83 = TextureSample(UVs=Add(A=Multiply(A=@n79, B=@n80), B=AppendVector(A=@n81, B=@n82.R))) {texture=&T_Inky_Smoke_Tile}
  n84 = Multiply(
   A=LinearInterpolate(Alpha=Power(Base=@n74)) {constA=0.3}
   B=Multiply(A=Multiply(A=@n76, B=@n78) {constA=1}, B=LinearInterpolate(Alpha=@n83.R) {constA=0.3, constB=1.5})
  ) {constA=1}
  n85 = ComponentMask(None=TextureCoordinate()) {r=true}
  n86 = OneMinus(None=@n85)
  n87 = Saturate(None=Multiply(A=Multiply(A=@n85, B=@n86)) {constB=2.5})
  n88 = Multiply(
   A=TextureCoordinate() {uTiling=4.2, vTiling=0.6}
   B=AppendVector(A=ScalarParameter() {defaultValue=0.8, parameterName=U_Scale}, B=Constant() {r=1})
  )
  n89 = ComponentMask(None=@n88) {r=true}
  n90 = OneMinus(None=ComponentMask(None=@n88) {g=true})
  n91 = Panner(Coordinate=Add(A=AppendVector(A=@n89, B=@n90)) {constB=-0.35}) {speedX=0.3}
  NamedRerouteDeclaration(None=Multiply(
    A=Multiply(A=@n84, B=@n87)
    B=TextureSample(UVs=@n91) {samplerSource=SSM_Wrap_WorldGroupSettings, texture=/Game/Effects/Fort_Effects/Textures/T_Smoke_Ring_Mask2}.R
   )) {name=Ray_Mask, nodeColor=<rgba 1 0.1343 0 1>}
  NamedRerouteUsage()
 ]
}
