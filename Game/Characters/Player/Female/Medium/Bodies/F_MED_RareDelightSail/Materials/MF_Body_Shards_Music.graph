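# Material function for the RareDelightSail character body, adding an animated smoke-noise underlayer and shimmering gem crystal shards to incoming material attributes.
# Crystals blend in as clear coat, faking interiors with refracted, rotating bump offset, plus scrolling rainbow-LUT chips and cracks that change with time of day.
# Event audio amplitude is exposed as a reroute.
# A further music-crystal emissive branch is built but unconnected to the sole output.
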MaterialFunction /Game/Characters/Player/Female/Medium/Bodies/F_MED_RareDelightSail/Materials/MF_Body_Shards_Music
graph {nodes [
  n1 = GetMaterialAttributes(None=FunctionInput() {inputName=MA, inputType=FunctionInput_MaterialAttributes}) {attributeGetTypes [#1, #2, #3]}
  n2 = NamedRerouteUsage()
  n3 = FunctionInput(Preview=Fresnel() {baseReflectFraction=0, exponent=1}) {inputName=Fresnel, inputType=FunctionInput_Scalar}
  n4 = NamedRerouteDeclaration(None=Power(Base=NamedRerouteDeclaration(None=@n3) {name=existingFresnel, nodeColor=<rgba 0.4766 1 0 1>})) {name=PoweredFresnel, nodeColor=<rgba 0.1719 0 1 1>}
  n5 = MaterialFunctionCall(Float3=PixelNormalWS()) {materialFunction=&BreakOutFloat3Components}
  n6 = TextureSample(UVs=Panner(Coordinate=AppendVector(A=@n4, B=@n5.B)) {speedY=0.1}) {samplerSource=SSM_Wrap_WorldGroupSettings, texture=&T_Inky_Smoke_Tile}
  n7 = LinearInterpolate(
   A=VectorParameter() {parameterName=UnderlayerCoreColor, defaultValue=<rgba 0.651 0.1867 0.4096 1>}.RGB
   B=VectorParameter() {parameterName=UnderlayerEdgeColor, defaultValue=<rgba 0.9062 0.1711 0.2078 1>}.RGB
   Alpha=@n6.R
  )
  n8 = NamedRerouteUsage()
  n9 = NamedRerouteUsage()
  n10 = Multiply(A=Multiply(A=@n6.R, B=NamedRerouteUsage()), B=NamedRerouteUsage())
  n11 = NamedRerouteUsage()
  n12 = GetMaterialAttributes(None=SetMaterialAttributes!(
    MaterialAttributes=@n1.MaterialAttributes
    BaseColor=LinearInterpolate(A=@n2, B=@n7, Alpha=@n8)
    EmissiveColor=LinearInterpolate(A=@n9, B=@n10, Alpha=@n11)
    Normal=NamedRerouteUsage()
   ) {attributeSetTypes [#1, #3, #4]}) {attributeGetTypes [#1]}
  n13 = NamedRerouteDeclaration(None=@n12.MaterialAttributes) {name=MA, nodeColor=<rgba 0.8516 0 1 1>}
  n14 = SetMaterialAttributes!(
   MaterialAttributes=@n13
   ShadingModel=ShadingModel() {shadingModel=MSM_ClearCoat}
   CustomData0=Constant() {r=1}
   CustomData1=Constant() {r=0.15}
   BaseColor=NamedRerouteUsage()
   Roughness=ScalarParameter() {defaultValue=0.2, group=GemCrystal, parameterName=InnerDiamondRoughness}
   Metallic=ScalarParameter() {defaultValue=1, group=GemCrystal, parameterName=innerdiamondMetallic}
   EmissiveColor=NamedRerouteUsage()
  ) {attributeSetTypes [
    #5
    #6
    #7
    #1
    #8
    #2
    #3
   ]}
  n15 = NamedRerouteDeclaration(None=BlendMaterialAttributes(A=@n13, B=@n14, Alpha=NamedRerouteUsage())) {name=Crystal Material, nodeColor=<rgba 0.0078 0 1 1>}
  FunctionOutput(None=@n15)
  n16 = TextureSampleParameter2D() {
   group=Texture Input
   parameterName=Normals
   samplerSource=SSM_Wrap_WorldGroupSettings
   samplerType=SAMPLERTYPE_Normal
   texture=/Engine/EngineMaterials/DefaultNormal
  }
  n17 = TextureCoordinate()
  n18 = Constant() {r=1}
  n19 = ScalarParameter() {defaultValue=2048, group=NormalBlur, parameterName=TextureRes}
  n20 = MaterialFunctionCall!(OffsetAmount=Multiply(A=Divide(A=@n18, B=@n19), B=ScalarParameter() {defaultValue=1, group=NormalBlur, parameterName=NormalBlurPixels})) {materialFunction=/Engine/Functions/Engine_MaterialFunctions01/ImageAdjustment/BlurSampleOffsets}
  n21 = TextureSampleParameter2D(UVs=Add(A=@n17, B=@n20."0,1")) {
   group=Texture Input
   parameterName=Normals
   samplerSource=SSM_Wrap_WorldGroupSettings
   samplerType=SAMPLERTYPE_Normal
   texture=&DefaultNormal
  }
  n22 = TextureSampleParameter2D(UVs=Add(A=@n17, B=@n20."1,0")) {
   group=Texture Input
   parameterName=Normals
   samplerSource=SSM_Wrap_WorldGroupSettings
   samplerType=SAMPLERTYPE_Normal
   texture=&DefaultNormal
  }
  n23 = TextureSampleParameter2D(UVs=Add(A=@n17, B=@n20."0,-1")) {
   group=Texture Input
   parameterName=Normals
   samplerSource=SSM_Wrap_WorldGroupSettings
   samplerType=SAMPLERTYPE_Normal
   texture=&DefaultNormal
  }
  n24 = TextureSampleParameter2D(UVs=Add(A=@n17, B=@n20."-1,0")) {
   group=Texture Input
   parameterName=Normals
   samplerSource=SSM_Wrap_WorldGroupSettings
   samplerType=SAMPLERTYPE_Normal
   texture=&DefaultNormal
  }
  n25 = Add(A=@n16.RGB, B=Add(A=Add(A=@n21.RGB, B=@n22.RGB), B=Add(A=@n23.RGB, B=@n24.RGB)))
  NamedRerouteDeclaration(None=QualitySwitch(Default=@n16.RGB, Epic=Divide(A=@n25) {constB=5})) {name=BlurredNormal, nodeColor=<rgba 0 1 0.6484 1>}
  n26 = NamedRerouteDeclaration(None=Multiply(A=@n12.BaseColor, B=ScalarParameter() {defaultValue=1, group=GemCrystal, parameterName=CrytalTintPremultiply})) {name=CrystalColor, nodeColor=<rgba 1 0 0.7969 1>}
  n27 = ScalarParameter() {group=GemCrystal, parameterName=BaseColorCrystalTint}
  n28 = NamedRerouteUsage()
  n29 = Multiply(A=TextureCoordinate(), B=ScalarParameter() {defaultValue=8, parameterName=CrystalInteriorTiling})
  n30 = Multiply(A=Time(), B=ScalarParameter() {defaultValue=1, parameterName=RotatorSpeed})
  n31 = MaterialFunctionCall!(
   Ray Direction=CameraVectorWS()
   Surface Normal=PixelNormalWS()
   Refractive Index Origin
   Refractive Index Target=ScalarParameter() {defaultValue=2, group=GemCrystal, parameterName=ref}
  ) {materialFunction=/Engine/Functions/Engine_MaterialFunctions01/Vectors/Refract}
  n32 = Transform(None=@n31.Vector) {transformSourceType=TRANSFORMSOURCE_World, transformType=TRANSFORM_Tangent}
  n33 = TextureSampleParameter2D(UVs=MaterialFunctionCall!(
    HeightTexture=Constant()
    TextureCoordinates=Rotator(Coordinate=@n29, Time=@n30)
    ReferencePlane=Constant() {r=1}
    HeightRatioInput=ScalarParameter() {defaultValue=0.5, group=GemCrystal, parameterName=rat}
    CameraVector=StaticSwitchParameter(True=@n32, False=@n31.Vector) {parameterName=UseTangentRefraction}
   ) {materialFunction=/Engine/Functions/Engine_MaterialFunctions02/Utility/BumpOffset_Advanced.BumpOffset_advanced}.UVs) {
   group=GemCrystal
   parameterName=GemInterior
   texture=/Game/Characters/Player/Female/Medium/Bodies/F_MED_RareDelightSail/Textures/T_RareDelightSail_BG
  }
  n34 = Multiply(
   A=NamedRerouteDeclaration(None=@n33.RGB) {name=GemRefraction, nodeColor=<rgba 0 1 0.3672 1>}
   B=ScalarParameter() {defaultValue=10, group=GemCrystal, parameterName=diamondBGBrightness}
  )
  NamedRerouteDeclaration(None=Multiply(A=LinearInterpolate(B=@n26, Alpha=@n27) {constA=1}, B=Multiply(A=@n28, B=@n34))) {name=BaseColorOUT, nodeColor=<rgba 0.8047 0 1 1>}
  n35 = NamedRerouteUsage()
  NamedRerouteDeclaration(None=Power(Base=OneMinus(None=@n35), Exp=ScalarParameter() {defaultValue=4, group=GemCrystal, parameterName=HighContrastFacingExp}) {constExponent=4}) {name=highContrastFacingAngle, nodeColor=<rgba 0 0.0859 1 1>}
  NamedRerouteDeclaration(None=SmoothStep(Value=@n33.A) {constMin=0.5}) {name=GemAlpha, nodeColor=<rgba 0.3594 0 1 1>}
  n36 = TextureSampleParameter2D() {
   group=GemCrystal
   parameterName=Fxmap
   samplerType=SAMPLERTYPE_Masks
   texture=/Game/Characters/Player/Female/Medium/Bodies/F_MED_RareDelightSail/Textures/T_RareDelightSail_Body_FX
  }
  NamedRerouteDeclaration(None=@n36.A) {name=Cracks Alpha, nodeColor=<rgba 1 0 0.5156 1>}
  n37 = Multiply(A=Time(), B=ScalarParameter() {defaultValue=0.5, group=GemCrystal, parameterName=LUTSpeed})
  n38 = TextureSampleParameter2D(UVs=Add(A=NamedRerouteUsage(), B=Frac(None=@n37))) {group=GemCrystal, parameterName=RainbowLUT, texture=/Game/Characters/Player/Female/Medium/Bodies/F_MED_RareDelightSail/Textures/T_MusicLUT}
  n39 = Multiply(A=Multiply(A=NamedRerouteUsage(), B=@n38.RGB), B=ScalarParameter() {defaultValue=3, group=GemCrystal, parameterName=ChromaticChipBrightness})
  n40 = Multiply(
   A=LinearInterpolate(B=NamedRerouteUsage(), Alpha=ScalarParameter() {group=GemCrystal, parameterName=EmissiveCrystalTint}) {constA=1}
   B=Multiply(A=NamedRerouteUsage(), B=@n39) {constB=0}
  )
  n41 = ScalarParameter() {defaultValue=0.5, group=GemCrystal, parameterName=HighContrastIntensity}
  n42 = Power(Base=NamedRerouteUsage())
  n43 = ScalarParameter() {defaultValue=2, group=GemCrystal, parameterName=Cracks}
  n44 = Divide(A=@n43)
  n45 = Add(
   A=Add(A=Multiply(A=@n40, B=@n41), B=@n39)
   B=Multiply(
    A=Multiply(A=@n38.RGB, B=@n42)
    B=MaterialFunctionCall(Value_Morning=@n43, Value_Day=@n43, Value_Evening=@n44, Value_Night=@n44) {materialFunction=&Scalar_Time_of_Day}.Result
   )
  )
  n46 = Multiply(
   A=Desaturation(None=@n45, Fraction=ScalarParameter() {defaultValue=0.2, group=GemCrystal, parameterName=NightDesaturation})
   B=ScalarParameter() {defaultValue=0.5, group=GemCrystal, parameterName=NightEmission}
  )
  NamedRerouteDeclaration(None=MaterialFunctionCall(Color_Morning=@n45, Color_Day=@n45, Color_Evening=@n46, Color_Night=@n46) {materialFunction=&Color_Time_of_Day}.Result) {name=EmissiveOUT, nodeColor=<rgba 0 0.6953 1 1>}
  NamedRerouteDeclaration(None=@n36.G) {name=ArmMask, nodeColor=<rgba 0 1 0.5313 1>}
  NamedRerouteDeclaration(None=@n36.R) {name=underlayerMask, nodeColor=<rgba 0.1484 0 1 1>}
  NamedRerouteDeclaration(None=@n1.BaseColor) {name=BasecolorIN, nodeColor=<rgba 1 0 0.2812 1>}
  NamedRerouteDeclaration(None=VectorParameter() {parameterName=UnderlayerEmis, defaultValue=<rgba 50 0 47.8021 1>}.RGB) {name=UnderLayerEM, nodeColor=<rgba 1 0 0.9844 1>}
  NamedRerouteDeclaration(None=@n1.EmissiveColor) {name=EmissiveIN, nodeColor=<rgba 0 0.4141 1 1>}
  NamedRerouteDeclaration(None=@n36.B) {name=Crystal Area, nodeColor=<rgba 0.6875 0 1 1>}
  n47 = CollectionParameter() {collection=&FortniteMaterialParameters, parameterName=EventAudio_Amplitude}
  NamedRerouteDeclaration(None=@n47) {name=MusicAmplitude, nodeColor=<rgba 0 1 0.2266 1>}
  n48 = GetMaterialAttributes(None=@n15) {attributeGetTypes [#3, #1]}
  n49 = MaterialFunctionCall!(Time, Divisor=Constant() {r=5}, Mod) {materialFunction=/Game/Effects/Fort_Effects/Materials/_Functions/MF_ModuloTime}
  n50 = TextureObject() {samplerType=SAMPLERTYPE_Normal, texture=/Game/Characters/Player/Female/Medium/Bodies/F_MED_RareDelightSail/Textures/T_Gem_N}
  n51 = ScalarParameter() {defaultValue=20, group=MusicCrystal, parameterName=WorldTiledTexture}
  n52 = Fresnel(Normal=Transform(None=MaterialFunctionCall!(TextureObject=@n50, TextureSize=@n51, PreskinnedLocalNormal, PreskinnedLocalPosition, Use High Quality Normals) {materialFunction=/Game/Global/MaterialFunctions/MF_PreskinnedWorldAlignedTextureNormal}.XYZ Texture)) {exponent=2}
  n53 = TextureSampleParameter2D(UVs=Multiply(A=Add(A=@n49.Clamped 0-1, B=@n52)) {constB=2}) {group=GemCrystal, parameterName=RainbowLUT, texture=&T_MusicLUT}
  n54 = Fresnel(ExponentIn=ScalarParameter() {defaultValue=1, group=MusicCrystal, parameterName=FresnelExp}) {baseReflectFraction=0, exponent=1}
  n55 = Multiply(
   A=Multiply(A=@n53.RGB, B=LinearInterpolate(Alpha=@n54) {constB=0.8})
   B=ScalarParameter() {defaultValue=8, group=MusicCrystal, parameterName=MusicCrystalIntensity}
  )
  n56 = NamedRerouteDeclaration(None=Add(A=NamedRerouteUsage(), B=NamedRerouteUsage())) {name=MusicArea, nodeColor=<rgba 0.8281 1 0 1>}
  n57 = ComponentMask(None=Panner() {speedY=0.3}) {g=true}
  n58 = Constant() {r=1}
  n59 = TextureSampleParameter2D(UVs=Multiply(A=AppendVector(A=@n57, B=@n58)) {constB=10}) {group=GemCrystal, parameterName=RainbowLUT, texture=&T_MusicLUT}
  n60 = TextureSampleParameter2D() {
   group=MusicCrystal
   parameterName=AdditionalTechArtMask
   samplerType=SAMPLERTYPE_Masks
   texture=/Game/Characters/Player/Female/Medium/Bodies/F_MED_RareDelightSail/Textures/T_AdditionalTechArtMask
  }
  n61 = Multiply(A=@n59.RGB, B=@n60.R)
  n62 = ScalarParameter() {defaultValue=1.5, group=MusicCrystal, parameterName=SecondEmissiveIntensity}
  n63 = Multiply(A=Add(A=Multiply(A=@n55, B=@n56), B=Multiply(A=@n61, B=@n62)), B=NamedRerouteUsage())
  n64 = Divide(A=@n63)
  n65 = Divide(A=@n48.BaseColor)
  n66 = NamedRerouteUsage()
  SetMaterialAttributes!(
   MaterialAttributes=@n48.MaterialAttributes
   EmissiveColor=Add(
    A=@n48.EmissiveColor
    B=MaterialFunctionCall(Color_Morning=@n63, Color_Day=@n63, Color_Evening=@n64, Color_Night=@n64) {materialFunction=&Color_Time_of_Day}.Result
   )
   BaseColor=LinearInterpolate(A=@n48.BaseColor, B=LinearInterpolate(A=@n48.BaseColor, B=@n65, Alpha=@n66), Alpha=NamedRerouteUsage())
  ) {attributeSetTypes [#3, #1]}
  Time() {bOverride_Period=true, period=0.5}
 ]}
