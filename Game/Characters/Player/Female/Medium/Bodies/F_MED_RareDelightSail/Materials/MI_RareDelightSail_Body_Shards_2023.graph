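# Character body material for the RareDelightSail female medium player skin, built from material attributes for skeletal meshes with morph targets and mesh deformer support.
# The shared character PreFX function feeds MF_Body_Shards_Music, which by its name adds a music-themed shard effect.
# That result passes through the shared PostFX function and then CINE_MF_CharPostFX.
# The shading model comes from the expression, covering Default Lit, Clear Coat and Subsurface Profile.
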Material /Game/Characters/Player/Female/Medium/Bodies/F_MED_RareDelightSail/Materials/MI_RareDelightSail_Body_Shards_2023
graph {
 properties {
  bUseMaterialAttributes=true
  bUsedWithMeshDeformer=true
  bUsedWithMorphTargets=true
  bUsedWithSkeletalMesh=true
  pixelDepthOffsetMode=PDOM_Legacy
  shadingModel=MSM_FromMaterialExpression
  usedShadingModels=Default Lit | Clear Coat | Subsurface Profile
 }
 nodes [
  n1 = MaterialFunctionCall() {materialFunction=&BaseFN_CharacterFunction_PreFX}
  n2 = MaterialFunctionCall(
   AO=@n1.AO
   Distance Blend=@n1.AO
   Fresnel=@n1.AO
   Gmap=@n1.AO
   Pre Skinned Local Position=@n1.AO
   Material=MaterialFunctionCall!(Fresnel=@n1.Fresnel, MA=@n1.Fresnel) {materialFunction=&/MF_Body_Shards_Music}.Result
  ) {materialFunction=&BaseFN_CharacterFunction_PostFX}
  MaterialFunctionCall(In=@n2.Result) {materialFunction=&CINE_MF_CharPostFX}
 ]
}
